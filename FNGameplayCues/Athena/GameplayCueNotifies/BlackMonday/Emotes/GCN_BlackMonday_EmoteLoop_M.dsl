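; Looping gameplay cue for the BlackMonday emote (M variant), running Cascade light and bat effects on the player.
; Loop start snaps to the mesh Root, then the AnimLightFixture timeline drives particle Scale and flickering Alpha, triggering bats via an event track.
; Removal plays a TurnOff timeline that winds down both effects and destroys the actor.
; Application and removal also toggle the emote flag on the backpack's Houston cape anim instance.
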(event EventOnLoopingStart(Cascade)
  (params (MyTarget :type Actor%20Object%20Reference) (Parameters :type Gameplay%20Cue%20Parameters%20Structure%20%28by%20ref%29) (ParticleComponents :type Array%20of%20Cascade%20Particle%20System%20Component%20Object%20References) (AudioComponents :type Array%20of%20Audio%20Component%20Object%20References))
  (bind _playermesh (Variables|Default|GetPlayerMesh))
  (Utilities|FlowControl|Sequence
    (:then_0
      (Utilities|IsValid MyTarget
        (:Is%20Valid
          (bind _asfort_player_pawn (Utilities|Casting|CastToFortPlayerPawn MyTarget))
          (bind _output_get (Variables|Default|SetPawn _asfort_player_pawn))
          (Variables|Default|SetPlayerMesh (Variables|/Script/Engine.Character|GetMesh _output_get)))))
    (:then_1
      (Transformation|SetActorLocation self (Transformation|GetSocketLocation (Variables|Default|GetPlayerMesh) "Root") false false)
      (Transformation|AttachComponentToComponent (Variables|Default|GetArrow) _playermesh "Root" "SnapToTarget" "SnapToTarget" "KeepRelative" true)
      (Transformation|AttachComponentToComponent (Variables|Default|GetP_BlackMonday_Emote_M_Init_02) _playermesh "Root" "SnapToTarget" "SnapToTarget" "KeepRelative" true)
      (Components|Activation|Activate (Variables|Default|GetP_BlackMonday_Emote_M_Init) false)
      (CallFunction|EnableFX))))

(event Custom|EnableFX
  (bind _p_blackmonday_emote_m_init (Variables|Default|GetP_BlackMonday_Emote_M_Init))
  (bind (_direction _scale _lightflicker) (Timeline|AnimLightFixture PlayFromStart)
    (:Update
      ;[1 1 1] Scale Object
      (Effects|Components|ParticleSystem|SetVectorParameter _p_blackmonday_emote_m_init "Scale" (Math|Conversions|ToVector(Float) _scale))
      (Effects|Components|ParticleSystem|SetFloatParameter _p_blackmonday_emote_m_init "Alpha" _lightflicker))
    (:BeginBats
      (Components|Activation|Activate (Variables|Default|GetP_BlackMonday_Emote_M_Init_02) false))))

(event EventOnRemoval(Cascade)
  (params (MyTarget :type Actor%20Object%20Reference) (Parameters :type Gameplay%20Cue%20Parameters%20Structure%20%28by%20ref%29) (ParticleComponents :type Array%20of%20Cascade%20Particle%20System%20Component%20Object%20References) (AudioComponents :type Array%20of%20Audio%20Component%20Object%20References) (BurstCameraShakeInstance :type Camera%20Shake%20Base%20Object%20Reference) (BurstDecalInstance :type Decal%20Actor%20Object%20Reference))
  (bind _p_blackmonday_emote_m_init (Variables|Default|GetP_BlackMonday_Emote_M_Init))
  (bind _p_blackmonday_emote_m_init_2 (Variables|Default|GetP_BlackMonday_Emote_M_Init))
  (Utilities|FlowControl|Sequence
    (:then_0
      (Timeline|AnimLightFixture Stop))
    (:then_1
      (bind (_direction _scale_down _alpha) (Timeline|TurnOff PlayFromStart)
        (:Update
          (Effects|Components|ParticleSystem|SetVectorParameter _p_blackmonday_emote_m_init "Scale" (Math|Conversions|ToVector(Float) _scale_down))
          (Effects|Components|ParticleSystem|SetFloatParameter _p_blackmonday_emote_m_init "Alpha" _alpha))
        (:Finished
          (Components|Activation|Deactivate (Variables|Default|GetP_BlackMonday_Emote_M_Init))
          (Actor|DestroyActor))
        (:Disable%20Bats
          (Components|Activation|Deactivate (Variables|Default|GetP_BlackMonday_Emote_M_Init_02)))))
    (:then_2
      (bind _asfort_player_pawn (Utilities|Casting|CastToFortPlayerPawn MyTarget))
      (bind _ascustom_character_part_anim_instance_houston_cape_m (Utilities|Casting|CastToCustomCharacterPartAnimInstance_Houston_Cape_M (Components|SkeletalMesh|GetAnimInstance (Customization|GetSkeletalMeshforPartType _asfort_player_pawn "Backpack" :!class /Script/FortniteGame.FortPlayerPawn))))
      (Variables|/Script/FortniteGame.CustomCharacterPartAnimInstance_Houston_Cape_M|SetbIsPlayingBlackMondayEmote false _ascustom_character_part_anim_instance_houston_cape_m))))

(event EventOnApplication(Cascade)
  (params (MyTarget :type Actor%20Object%20Reference) (Parameters :type Gameplay%20Cue%20Parameters%20Structure%20%28by%20ref%29) (ParticleComponents :type Array%20of%20Cascade%20Particle%20System%20Component%20Object%20References) (AudioComponents :type Array%20of%20Audio%20Component%20Object%20References) (BurstCameraShakeInstance :type Camera%20Shake%20Base%20Object%20Reference) (BurstDecalInstance :type Decal%20Actor%20Object%20Reference))
  (bind _asfort_player_pawn (Utilities|Casting|CastToFortPlayerPawn MyTarget))
  (bind _ascustom_character_part_anim_instance_houston_cape_m (Utilities|Casting|CastToCustomCharacterPartAnimInstance_Houston_Cape_M (Components|SkeletalMesh|GetAnimInstance (Customization|GetSkeletalMeshforPartType _asfort_player_pawn "Backpack" :!class /Script/FortniteGame.FortPlayerPawn))))
  (Variables|/Script/FortniteGame.CustomCharacterPartAnimInstance_Houston_Cape_M|SetbIsPlayingBlackMondayEmote true _ascustom_character_part_anim_instance_houston_cape_m))

(fn ConstructionScript)
